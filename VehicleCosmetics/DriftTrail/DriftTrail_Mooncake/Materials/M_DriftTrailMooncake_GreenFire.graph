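# Translucent, unlit, two-sided Niagara ribbon material for the green fire of the Mooncake drift trail.
# Particle color feeds emissive adjustments; opacity comes from a panning fire-wave texture, warped by V-weighted caustics and boosted by scrolling noise.
# A sine along U, particle alpha, stepped noise dissolve and depth fade shape that opacity.
# The dissolve's add and subtract cancel the DissolveHigh step, and the OffsetX/Y/Z transform feeds nothing shown.
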Material /VehicleCosmetics/DriftTrail/DriftTrail_Mooncake/Materials/M_DriftTrailMooncake_GreenFire
graph {
 expressions {
  MaterialExpressionAdd_16 {
   class=MaterialExpressionAdd
   inputs [
    A {from=MaterialExpressionTextureSampleParameter2D_4, output=G}
    B {from=MaterialExpressionPower_25}
   ]
  }
  MaterialExpressionAdd_17 {
   class=MaterialExpressionAdd
   inputs [
    A {from=MaterialExpressionPanner_24}
    B {from=MaterialExpressionAppendVector_4}
   ]
  }
  MaterialExpressionAdd_18 {
   class=MaterialExpressionAdd
   inputs [
    A {from=MaterialExpressionTextureCoordinate_40}
    B {from=MaterialExpressionMultiply_0}
   ]
  }
  MaterialExpressionAdd_19 {
   class=MaterialExpressionAdd
   inputs [
    A {from=MaterialExpressionStep_2}
    B {from=MaterialExpressionSubtract_4}
   ]
  }
  MaterialExpressionAppendVector_4 {
   class=MaterialExpressionAppendVector
   inputs [
    A {from=MaterialExpressionScalarParameter_14, output=""}
    B {from=MaterialExpressionConstant_43, output=""}
   ]
  }
  MaterialExpressionComponentMask_0 {class=MaterialExpressionComponentMask, inputs [None {from=MaterialExpressionTextureCoordinate_0}], properties {g=true}}
  MaterialExpressionComponentMask_17 {class=MaterialExpressionComponentMask, inputs [None {from=MaterialExpressionTextureCoordinate_43}], properties {g=true}}
  MaterialExpressionComponentMask_18 {class=MaterialExpressionComponentMask, inputs [None {from=MaterialExpressionTextureCoordinate_46}], properties {r=true}}
  MaterialExpressionComponentMask_19 {class=MaterialExpressionComponentMask, inputs [None {from=MaterialExpressionTextureCoordinate_47}], properties {g=true}}
  MaterialExpressionComponentMask_6 {class=MaterialExpressionComponentMask, inputs [None {from=MaterialExpressionTextureCoordinate_41}], properties {g=true}}
  MaterialExpressionConstantClamp_8 {class=MaterialExpressionSaturate, inputs [None {from=MaterialExpressionMultiply_62}]}
  MaterialExpressionConstantClamp_9 {class=MaterialExpressionSaturate, inputs [None {from=MaterialExpressionAdd_19}]}
  MaterialExpressionConstant_43 {class=MaterialExpressionConstant}
  MaterialExpressionConvert_0 {
   class=MaterialExpressionConvert
   inputs [
    None {from=MaterialExpressionScalarParameter_4}
    None {from=MaterialExpressionScalarParameter_5}
    None {from=MaterialExpressionScalarParameter_6}
   ]
   properties {
    nodeName=Make Float3
    convertInputs [
     {
      type=Scalar
      expressionInput {
       expression=&self:MaterialExpressionScalarParameter_4
       inputName=None
       mask=0
       maskA=0
       maskB=0
       maskG=0
       maskR=0
       outputIndex=0
      }
     }
     {
      type=Scalar
      expressionInput {
       expression=&self:MaterialExpressionScalarParameter_5
       inputName=None
       mask=0
       maskA=0
       maskB=0
       maskG=0
       maskR=0
       outputIndex=0
      }
     }
     {
      type=Scalar
      expressionInput {
       expression=&self:MaterialExpressionScalarParameter_6
       inputName=None
       mask=0
       maskA=0
       maskB=0
       maskG=0
       maskR=0
       outputIndex=0
      }
     }
    ]
    convertOutputs [{type=Vector3}]
   }
  }
  MaterialExpressionMaterialFunctionCall_0 {
   class=MaterialExpressionMaterialFunctionCall
   inputs [Emissive {from=MaterialExpressionReroute_1}]
   properties {materialFunction=&MF_FX_EmissiveAdjustments}
  }
  MaterialExpressionMaterialFunctionCall_1 {
   class=MaterialExpressionMaterialFunctionCall
   inputs [
    Opacity In {from=MaterialExpressionMultiply_60}
    DepthFade Distance {from=MaterialExpressionScalarParameter_0}
   ]
   properties {materialFunction=&MF_VCVFX_DepthFade}
  }
  MaterialExpressionMultiply_0 {
   class=MaterialExpressionMultiply
   inputs [
    A {from=MaterialExpressionMultiply_65}
    B {from=MaterialExpressionPower_0}
   ]
  }
  MaterialExpressionMultiply_1 {
   class=MaterialExpressionMultiply
   inputs [
    A {from=MaterialExpressionComponentMask_0}
    B {from=""}
   ]
   properties {constB=2}
  }
  MaterialExpressionMultiply_15 {
   class=MaterialExpressionMultiply
   inputs [
    A {from=MaterialExpressionTextureSample_21, output=R}
    B {from=MaterialExpressionMultiply_72}
   ]
  }
  MaterialExpressionMultiply_16 {
   class=MaterialExpressionMultiply
   inputs [
    A {from=MaterialExpressionComponentMask_17}
    B {from=MaterialExpressionMultiply_71}
   ]
  }
  MaterialExpressionMultiply_17 {
   class=MaterialExpressionMultiply
   inputs [
    A {from=MaterialExpressionMultiply_15}
    B {from=MaterialExpressionMultiply_16}
   ]
  }
  MaterialExpressionMultiply_18 {
   class=MaterialExpressionMultiply
   inputs [
    A {from=MaterialExpressionScalarParameter_11}
    B {from=MaterialExpressionTextureSample_20, output=R}
   ]
  }
  MaterialExpressionMultiply_19 {
   class=MaterialExpressionMultiply
   inputs [
    A {from=MaterialExpressionPower_26}
    B {from=""}
   ]
   properties {constB=20}
  }
  MaterialExpressionMultiply_20 {
   class=MaterialExpressionMultiply
   inputs [
    A {from=MaterialExpressionTime_4}
    B {from=MaterialExpressionScalarParameter_12}
   ]
  }
  MaterialExpressionMultiply_22 {
   class=MaterialExpressionMultiply
   inputs [
    A {from=MaterialExpressionVertexColor_4, output=A}
    B {from=MaterialExpressionAdd_16}
   ]
  }
  MaterialExpressionMultiply_60 {
   class=MaterialExpressionMultiply
   inputs [
    A {from=MaterialExpressionConstantClamp_8}
    B {from=MaterialExpressionConstantClamp_9}
   ]
  }
  MaterialExpressionMultiply_62 {
   class=MaterialExpressionMultiply
   inputs [
    A {from=MaterialExpressionMultiply_22}
    B {from=MaterialExpressionSine_4}
   ]
  }
  MaterialExpressionMultiply_65 {
   class=MaterialExpressionMultiply
   inputs [
    A {from=MaterialExpressionMultiply_19}
    B {from=MaterialExpressionMultiply_18}
   ]
  }
  MaterialExpressionMultiply_71 {
   class=MaterialExpressionMultiply
   inputs [
    A {from=MaterialExpressionTextureSampleParameter2D_4, output=R}
    B {from=""}
   ]
   properties {constB=4}
  }
  MaterialExpressionMultiply_72 {
   class=MaterialExpressionMultiply
   inputs [
    A {from=MaterialExpressionPower_27}
    B {from=""}
   ]
   properties {constB=10}
  }
  MaterialExpressionMultiply_74 {
   class=MaterialExpressionMultiply
   inputs [
    A {from=MaterialExpressionTextureSample_24, output=R}
    B {from=MaterialExpressionTextureSample_23, output=R}
   ]
  }
  MaterialExpressionOneMinus_0 {class=MaterialExpressionOneMinus, inputs [None {from=MaterialExpressionComponentMask_19}]}
  MaterialExpressionOneMinus_4 {class=MaterialExpressionOneMinus, inputs [None {from=MaterialExpressionComponentMask_6}]}
  MaterialExpressionPanner_24 {
   class=MaterialExpressionPanner
   inputs [
    Coordinate {from=MaterialExpressionAdd_18}
    Time {from=MaterialExpressionMultiply_20}
    Speed {from=""}
   ]
   properties {speedX=-0.85}
  }
  MaterialExpressionPanner_25 {
   class=MaterialExpressionPanner
   inputs [
    Coordinate {from=MaterialExpressionTextureCoordinate_42}
    Time {from=""}
    Speed {from=""}
   ]
   properties {speedY=1}
  }
  MaterialExpressionPanner_26 {
   class=MaterialExpressionPanner
   inputs [
    Coordinate {from=MaterialExpressionTextureCoordinate_44}
    Time {from=""}
    Speed {from=""}
   ]
   properties {speedX=0.2}
  }
  MaterialExpressionPanner_27 {
   class=MaterialExpressionPanner
   inputs [
    Coordinate {from=MaterialExpressionTextureCoordinate_45}
    Time {from=""}
    Speed {from=""}
   ]
   properties {speedX=-0.35}
  }
  MaterialExpressionPanner_28 {
   class=MaterialExpressionPanner
   inputs [
    Coordinate {from=MaterialExpressionTextureCoordinate_48}
    Time {from=""}
    Speed {from=""}
   ]
   properties {speedX=-1}
  }
  MaterialExpressionPanner_29 {
   class=MaterialExpressionPanner
   inputs [
    Coordinate {from=MaterialExpressionTextureCoordinate_49}
    Time {from=""}
    Speed {from=""}
   ]
   properties {speedX=-0.5}
  }
  MaterialExpressionPower_0 {
   class=MaterialExpressionPower
   inputs [
    Base {from=MaterialExpressionMultiply_1}
    Exp {from=""}
   ]
   properties {constExponent=4}
  }
  MaterialExpressionPower_25 {
   class=MaterialExpressionPower
   inputs [
    Base {from=MaterialExpressionMultiply_17}
    Exp {from=""}
   ]
   properties {constExponent=0.4}
  }
  MaterialExpressionPower_26 {
   class=MaterialExpressionPower
   inputs [
    Base {from=MaterialExpressionOneMinus_4}
    Exp {from=""}
   ]
   properties {constExponent=1.5}
  }
  MaterialExpressionPower_27 {
   class=MaterialExpressionPower
   inputs [
    Base {from=MaterialExpressionTextureSample_22, output=R}
    Exp {from=""}
   ]
   properties {constExponent=5}
  }
  MaterialExpressionPower_28 {
   class=MaterialExpressionPower
   inputs [
    Base {from=MaterialExpressionComponentMask_19}
    Exp {from=MaterialExpressionScalarParameter_13}
   ]
  }
  MaterialExpressionPower_29 {
   class=MaterialExpressionPower
   inputs [
    Base {from=MaterialExpressionOneMinus_0}
    Exp {from=MaterialExpressionScalarParameter_10}
   ]
  }
  MaterialExpressionReroute_0 {class=MaterialExpressionReroute, inputs [None {from=MaterialExpressionMultiply_74}]}
  MaterialExpressionReroute_1 {class=MaterialExpressionReroute, inputs [None {from=MaterialExpressionVertexColor_4, output=RGB}]}
  MaterialExpressionScalarParameter_0 {class=MaterialExpressionScalarParameter, properties {defaultValue=60, parameterName="DepthFade Distance (S)"}}
  MaterialExpressionScalarParameter_10 {class=MaterialExpressionScalarParameter, properties {defaultValue=1.5, parameterName=DissolveLow}}
  MaterialExpressionScalarParameter_11 {class=MaterialExpressionScalarParameter, properties {defaultValue=0.01, parameterName=Distort}}
  MaterialExpressionScalarParameter_12 {class=MaterialExpressionScalarParameter, properties {defaultValue=1, parameterName=Velocity}}
  MaterialExpressionScalarParameter_13 {class=MaterialExpressionScalarParameter, properties {defaultValue=4, parameterName=DissolveHigh}}
  MaterialExpressionScalarParameter_14 {class=MaterialExpressionScalarParameter, properties {defaultValue=0.05, parameterName=OffsetTrail}}
  MaterialExpressionScalarParameter_4 {class=MaterialExpressionScalarParameter, properties {parameterName=OffsetX}}
  MaterialExpressionScalarParameter_5 {class=MaterialExpressionScalarParameter, properties {parameterName=OffsetY}}
  MaterialExpressionScalarParameter_6 {class=MaterialExpressionScalarParameter, properties {parameterName=OffsetZ}}
  MaterialExpressionSine_4 {class=MaterialExpressionSine, inputs [None {from=MaterialExpressionComponentMask_18}], properties {period=2}}
  MaterialExpressionStep_2 {
   class=MaterialExpressionStep
   inputs [
    Y {from=MaterialExpressionReroute_0, output=""}
    X {from=MaterialExpressionPower_28, output=""}
   ]
   properties {}
  }
  MaterialExpressionStep_3 {
   class=MaterialExpressionStep
   inputs [
    Y {from=MaterialExpressionMultiply_74, output=""}
    X {from=MaterialExpressionPower_29, output=""}
   ]
   properties {}
  }
  MaterialExpressionSubtract_4 {
   class=MaterialExpressionSubtract
   inputs [
    A {from=MaterialExpressionStep_3}
    B {from=MaterialExpressionStep_2}
   ]
  }
  MaterialExpressionTextureCoordinate_0 {class=MaterialExpressionTextureCoordinate}
  MaterialExpressionTextureCoordinate_40 {class=MaterialExpressionTextureCoordinate}
  MaterialExpressionTextureCoordinate_41 {class=MaterialExpressionTextureCoordinate}
  MaterialExpressionTextureCoordinate_42 {class=MaterialExpressionTextureCoordinate, properties {uTiling=4, vTiling=0.5}}
  MaterialExpressionTextureCoordinate_43 {class=MaterialExpressionTextureCoordinate}
  MaterialExpressionTextureCoordinate_44 {class=MaterialExpressionTextureCoordinate, properties {uTiling=4, vTiling=0.5}}
  MaterialExpressionTextureCoordinate_45 {class=MaterialExpressionTextureCoordinate, properties {uTiling=4}}
  MaterialExpressionTextureCoordinate_46 {class=MaterialExpressionTextureCoordinate}
  MaterialExpressionTextureCoordinate_47 {class=MaterialExpressionTextureCoordinate}
  MaterialExpressionTextureCoordinate_48 {class=MaterialExpressionTextureCoordinate, properties {uTiling=2, vTiling=0.5}}
  MaterialExpressionTextureCoordinate_49 {class=MaterialExpressionTextureCoordinate, properties {uTiling=10, unMirrorU=true}}
  MaterialExpressionTextureSampleParameter2D_4 {
   class=MaterialExpressionTextureSampleParameter2D
   inputs [
    UVs {from=MaterialExpressionAdd_17, output=""}
    Apply View MipBias {from="", output=""}
   ]
   properties {parameterName=Texture, texture=/VehicleCosmetics/DriftTrail/DriftTrail_Mooncake/Textures/T_Mooncake_FireWave_RGB}
  }
  MaterialExpressionTextureSample_20 {
   class=MaterialExpressionTextureSample
   inputs [
    UVs {from=MaterialExpressionPanner_25}
    Tex {from=""}
    Apply View MipBias {from=""}
   ]
   properties {texture=/DealershipSharedLibrary/FX/Textures/Noise/T_VCVFX_CausticsNoise_A}
  }
  MaterialExpressionTextureSample_21 {
   class=MaterialExpressionTextureSample
   inputs [
    UVs {from=MaterialExpressionPanner_26}
    Tex {from=""}
    Apply View MipBias {from=""}
   ]
   properties {texture=&T_VCVFX_CausticsNoise_A}
  }
  MaterialExpressionTextureSample_22 {
   class=MaterialExpressionTextureSample
   inputs [
    UVs {from=MaterialExpressionPanner_27}
    Tex {from=""}
    Apply View MipBias {from=""}
   ]
   properties {texture=/VehicleCosmetics/DriftTrail/DriftTrail_Mooncake/Textures/T_Mooncake_Fire_N}
  }
  MaterialExpressionTextureSample_23 {
   class=MaterialExpressionTextureSample
   inputs [
    UVs {from=MaterialExpressionPanner_28}
    Tex {from=""}
    Apply View MipBias {from=""}
   ]
   properties {texture=/DealershipSharedLibrary/FX/Textures/Noise/T_VCVFX_CellNoiseSkins_A}
  }
  MaterialExpressionTextureSample_24 {
   class=MaterialExpressionTextureSample
   inputs [
    UVs {from=MaterialExpressionPanner_29}
    Tex {from=""}
    Apply View MipBias {from=""}
   ]
   properties {texture=&T_VCVFX_CausticsNoise_A}
  }
  MaterialExpressionTime_4 {class=MaterialExpressionTime}
  MaterialExpressionTransform_0 {class=MaterialExpressionTransform, inputs [None {from=MaterialExpressionConvert_0}], properties {transformSourceType=TRANSFORMSOURCE_Local}}
  MaterialExpressionVertexColor_4 {class=MaterialExpressionParticleColor, inputs [], properties {}}
 }
 properties {
  bAutomaticallySetUsageInEditor=false
  bUsedWithNiagaraRibbons=true
  blendMode=BLEND_Translucent
  shadingModel=MSM_Unlit
  twoSided=true
  usedShadingModels=Unlit
 }
}
